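# Unlit masked material for the moon mesh particles in the GrimHound skydiving contrail Niagara effect.
# Emissive tints a moon texture blue to near-white, modulated by inverted fresnel and a time-scrolled ramp, with DiveLerp and HDRMultMax dynamic parameters driving intensity.
# Other nodes build a sine-animated UV mask, isGlow blends with fresnel, and a near-camera fade through temporal AA dither.
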Material /BRCosmetics/Athena/Effects/Particles/Skydiving/NS_Contrail_GrimHound/M_Contrail_GrimHound_Moon
graph {
 properties {
  bAutomaticallySetUsageInEditor=false
  bUsedWithNiagaraMeshParticles=true
  blendMode=BLEND_Masked
  shadingModel=MSM_Unlit
  usedShadingModels=Unlit
 }
 nodes [
  Constant4Vector() {constant=<rgba 0.016 0.1559 0.2346 1>}
  n1 = Constant3Vector() {constant=<rgba 0.02 0.2 0.3 1>}
  n2 = Constant3Vector() {constant=<rgba 0.9 0.95 1 0>}
  n3 = TextureSample(UVs=TextureCoordinate()) {texture=&/T_FX_Contrail_GrimHound_Moon}
  n4 = Constant() {r=0.05}
  n5 = Saturate(None=Multiply(A=MaterialFunctionCall(In=@n3.R, Contrast=@n4) {materialFunction=&CheapContrast}.Result) {constB=1.85})
  LinearInterpolate(
   A=Multiply(A=LinearInterpolate(A=@n1, B=@n2, Alpha=@n5)) {constB=1.8}
   B=Constant3Vector() {constant=<rgba 0.016 0.1559 0.2346 1>}
   Alpha=ScalarParameter() {parameterName=isGlow}
  )
  n6 = Constant() {r=-1}
  LinearInterpolate(A=@n6, B=ScalarParameter() {defaultValue=-1, parameterName=GlowRampOffset}, Alpha=ScalarParameter() {parameterName=isGlow}) {constA=-1, constB=-1.025}
  n7 = ScalarParameter() {defaultValue=0.2, parameterName=MaskPositionLow}
  n8 = ScalarParameter() {parameterName=isGlow}
  LinearInterpolate(A=@n7, B=ScalarParameter() {defaultValue=0.2, parameterName=GlowMaskPositionLow}, Alpha=@n8)
  n9 = ScalarParameter() {defaultValue=0.57, parameterName=MaskPositionHigh}
  LinearInterpolate(A=@n9, B=ScalarParameter() {defaultValue=0.57, parameterName=GlowMaskPositionHigh}, Alpha=@n8)
  n10 = Multiply(A=Time(), B=ScalarParameter() {defaultValue=0.1, parameterName=MaskSpeed})
  n11 = LinearInterpolate(
   A=@n7
   B=@n9
   Alpha=MaterialFunctionCall(Input=Sine(None=@n10), Input Low=Constant() {r=-1}, Input High=Constant() {r=1}, Target Low=Constant(), Target High=Constant() {r=1}) {materialFunction=&RemapValueRange}.Result
  )
  n12 = ComponentMask(None=Add(A=TextureCoordinate(), B=Add(A=@n11, B=@n6))) {r=true}
  n13 = OneMinus(None=@n12)
  LinearInterpolate(A=@n13, B=@n12, Alpha=ScalarParameter() {parameterName=isGlow})
  n14 = TextureSample(UVs=TextureCoordinate()) {texture=&T_FX_Contrail_GrimHound_Moon}
  n15 = ScalarParameter() {defaultValue=1.1, parameterName=EmissiveLightFresnelBrightness}
  n16 = ScalarParameter() {defaultValue=3, parameterName=EmissiveLightFresnelExponent}
  n17 = MaterialFunctionCall(Float2=TextureCoordinate()) {materialFunction=&BreakOutFloat2Components}
  n18 = Frac(None=Multiply(A=Time()) {constB=0.25})
  n19 = Multiply(A=Add(A=@n17.R, B=Multiply(A=@n18) {constB=-1}))
  n20 = Multiply(A=@n17.G)
  n21 = Multiply(A=TextureSample(UVs=MaterialFunctionCall(X=@n19, Y=@n20) {materialFunction=&MakeFloat2}.Result) {texture=/Game/Global/Textures/Default/Masks_and_Gradients/T_Ramp_Virt}.RGB) {constB=3}
  n22 = DynamicParameter() {defaultValue=<rgba 1 1.5 1 1>, paramNames [DiveLerp,HDRMultMax,Param3,Param4]}
  n23 = LinearInterpolate(A=Saturate(None=Power(Base=@n21) {constExponent=3}), Alpha=@n22.DiveLerp)
  n24 = LinearInterpolate(Alpha=@n23) {constA=0.1}
  n25 = Multiply(
   A=Saturate(None=MaterialFunctionCall(Brightness=@n15, Exponent=@n16) {materialFunction=&MF_Fresnel_FX}.Fresnel Inverted)
   B=@n24
  )
  n26 = LinearInterpolate(Alpha=Saturate(None=Multiply(A=@n25, B=@n22.DiveLerp))) {constB=0.5}
  n27 = LinearInterpolate(
   A=Constant3Vector() {constant=<rgba 0.02 0.2 0.3 1>}
   B=Constant3Vector() {constant=<rgba 0.9 0.95 1 0>}
   Alpha=Saturate(None=Add(A=@n14.RGB, B=@n26))
  )
  n28 = LinearInterpolate(A=ScalarParameter() {defaultValue=0.5, parameterName=HDRMin}, B=@n22.HDRMultMax, Alpha=@n23)
  MaterialFunctionCall(Emissive=Multiply(
    A=MaterialFunctionCall(LDR Input=@n27, HDR Intensity Multiplier=@n28) {materialFunction=&DeriveHDRfromLDR}.RemappedRange
    B=@n24
   )) {materialFunction=&MF_FX_EmissiveAdjustments}
  MaterialFunctionCall(Alpha Threshold=MaterialFunctionCall(Distance=ScalarParameter() {defaultValue=200, parameterName=DitherDistance}) {materialFunction=&Near_Camera_Fade}.Result) {materialFunction=&DitherTemporalAA}
  n29 = Saturate(None=Power(Base=@n13, Exp=ScalarParameter() {defaultValue=80, parameterName=MaskPower}) {constExponent=80})
  Multiply(A=@n3.R, B=@n29)
  Multiply(A=@n14.RGB, B=@n26)
  n30 = DynamicParameter() {defaultValue=<rgba 1 1.5 1 1>, paramNames [DiveLerp,HDRMultMax,Param3,Param4]}
  n31 = Saturate(None=MaterialFunctionCall(
    Brightness=ScalarParameter() {defaultValue=1, parameterName=FresnelBrightness}
    Exponent=ScalarParameter() {defaultValue=2, parameterName=FresnelExponent}
   ) {materialFunction=&MF_Fresnel_FX}.Fresnel)
  n32 = ScalarParameter() {parameterName=isGlow}
  Multiply(A=ParticleColor().A, B=Multiply(A=LinearInterpolate(A=@n29, Alpha=@n30.DiveLerp), B=LinearInterpolate(B=@n31, Alpha=@n32) {constA=1}))
  ScalarParameter() {defaultValue=1.5, parameterName=HDRMax}
 ]
}
