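# Niagara skydiving contrail for the GrimHound cosmetic, layering a glowing moon, corona, fire shell and trails in dark teal and cyan.
# Persistent burst-spawned mesh, sprite and light particles follow skeletal bones for limb trails, or a shared SkullPosition for moon layers.
# Dive value and normalized Timeline drive scale, alpha and material parameters, with camera-angle fades and time-looped pulses.
# FrontEnd modifiers retune sizes, offsets and spawn counts outside gameplay.
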NiagaraSystem /BRCosmetics/Athena/Effects/Particles/Skydiving/NS_Contrail_GrimHound/NS_Contrail_GrimHound
dependencies {
 usedDataInterfaces [
  &NiagaraDataInterfaceCurve
  &NiagaraDataInterfaceColorCurve
  &NiagaraDataInterfaceCamera
  &NiagaraDataInterfaceArrayFloat
 ]
 usedDynamicInputs [&SetBoolByFloatComparison]
 usedModules [
  &self:SystemSpawnScript.SystemScriptSource.SystemScriptGraph.NiagaraNodeAssignment_5.SetVariables_#1
  &SampleSkeletalMesh
  /Game/Effects/Niagara/Modules/Skydiving/FN_DivingAttributes/v5/FN_DivingAttributes
  &self:SystemSpawnScript.SystemScriptSource.SystemScriptGraph.NiagaraNodeAssignment_0.SetVariables_#2
  /Game/Effects/Niagara/Modules/Contrails/FN_Contrail_Setup
  &SystemState
  &Timeline
  &self:SystemSpawnScript.SystemScriptSource.SystemScriptGraph.NiagaraNodeAssignment_8.SetVariables_#3
  &self:SystemSpawnScript.SystemScriptSource.SystemScriptGraph.NiagaraNodeAssignment_14.SetVariables_#4
 ]
}
emitters {
 Corona {
  emitterUpdateScript [
   &EmitterState(Life Cycle Mode=::NewEnumerator1, Loop Behavior=::NewEnumerator1, Loop Duration Mode=::NewEnumerator1, UseLoopDelay=false)
   &SpawnBurst_Instantaneous(
    Spawn Probability=&Lerp_Float(A=0, B=1, Alpha=&FN_FrontEndFloatModifier(Gameplay=1, Front End=1))
    Use Spawn Probability=false
    UseLoopCountLimit=false
   )
  ]
  particleSpawnScript [
   &PartitionParticles(Number Of Partitions=5)
   &InitializeParticle(
    Color=<rgba 0.0015 0.0152 0.0212 1>
    Mesh Scale=&Multiply_VectorByFloat(Vector=<1 1.1 1.1>, Float=0.52)
    Mesh Scale Mode=::NewEnumerator4
    Sprite Size Mode=::NewEnumerator3
    Uniform Sprite Size=65
    UsePositionOffset=false
    Write Material Random=true
   )
   &CameraOffset(Camera Offset Amount=10)
   &InitialMeshOrientation(
    Mesh Orientation Mode=::NewEnumerator2
    Orientation Vector=&RotateVectorByQuaternion(Vector To Rotate=<1 0 0>, Quaternion=$Engine.Owner.Rotation)
    Use Rotation Vector=false
   ) {enabled=false}
   &Color(Color=<rgba 0.0015 0.0152 0.0212 1>)
  ]
  particleUpdateScript [
   &SkeletalMeshLocation(
    Bone / Socket Index=$Particles.PartitionParticles.Partition
    Bone Sampling Mode=::NewEnumerator4
    MaskSpawnGroup=false
    Orientation Sampling=::NewEnumerator1
    Sampled Position Offset=&SelectVectorFromArray(
     Direct Array Index=$Particles.PartitionParticles.Partition
     Vector Selection Array=DataInterface() {propertyValues {FloatData [<0 0 0>, <0 0 0>]}}
    )
    Skeletal Mesh=$System.Skeletal Mesh
   )
   &ParticleState(Kill Particles When Lifetime Has Elapsed=false, Let Infinitely Lived Particles Die When Emitter Deactivates=true)
   &ScaleMeshSize()
   &DynamicMaterialParameters(
    Index 0 Param 1=&SelectFloatFromArray(
     Direct Array Index=$Particles.PartitionParticles.Partition
     Float Selection Array=DataInterface() {propertyValues {FloatData [0.8,0.6,0.2,0.1]}}
    )
    Index 0 Param 2=0.2
    Index 0 Param 4=0.5
    Module.Param0WriteEnabled=false
    Module.Param1WriteEnabled=true
    Module.Param2WriteEnabled=false
    Module.Param3WriteEnabled=false
   ) {enabled=false, module=DynamicMaterialParameters002}
   /SetVariables(Particles.Position=&TransformPosition(
     Position=&AddVectorToPosition(Position=Linked() {linkedVariable {name=System.SkullPosition, type=Vector3f}}, Vector=<0 0 -30>)
     Source Space=::World
     Destination Space=::Local
    )) {isSetParameters=true, module=SetVariables_#5}
   &ScaleColor(
    Scale Alpha=&Multiply_Float(
     A=&Multiply_Float(
      A=3
      B=&Power(
       a=&DotProduct(
        Vector A=&ReturnCameraVector(Camera Vectors=::NewEnumerator1, Coordinate Space=::Simulation)
        Vector B=<0 0 -1>
       )
       b=3
      )
     )
     B=&Lerp_Float(
      A=0
      B=1
      Alpha=&FloatFromCurve(
       FloatCurve=DataInterface() {propertyValues {Curve {
          defaultValue=0
          keys [
           {interpMode=RCIM_Cubic, tangentMode=RCTM_User, time=0, value=0}
           {arriveTangent=2.7352, interpMode=RCIM_Cubic, leaveTangent=2.7352, tangentMode=RCTM_User, time=1, value=1}
          ]
         }}}
       CurveIndex=$System.Timeline.NormalizedTime
      )
     )
    )
    ScaleA=true
    ScaleRGB=true
   ) {enabled=false, module=ScaleColor003}
   &ScaleColor(
    Scale Alpha=&Multiply_Float(
     A=&Multiply_Float(
      A=&FloatFromCurve(
       FloatCurve=DataInterface() {propertyValues {Curve {
          defaultValue=0
          keys [
           {arriveTangent=0.2035, interpMode=RCIM_Cubic, leaveTangent=0.2035, tangentMode=RCTM_User, time=0.7, value=0}
           {interpMode=RCIM_Cubic, tangentMode=RCTM_User, time=1, value=1.25}
          ]
         }}}
       CurveIndex=&DotProduct(
        Vector A=&ReturnCameraVector(
         Camera Vectors=::NewEnumerator1
         Coordinate Space=::Simulation
         CameraQuery=DataInterface() {propertyValues {bRequireCurrentFrameData=false}}
        )
        Vector B=<0 0 -1>
       )
      )
      B=1
     )
     B=&Lerp_Float(
      A=0
      B=10
      Alpha=&FloatFromCurve(
       FloatCurve=DataInterface() {propertyValues {Curve {
          defaultValue=0
          keys [
           {interpMode=RCIM_Cubic, tangentMode=RCTM_User, time=0.5, value=0}
           {arriveTangent=2.7352, interpMode=RCIM_Cubic, leaveTangent=2.7352, tangentMode=RCTM_User, time=1, value=1}
          ]
         }}}
       CurveIndex=$System.Timeline.NormalizedTime
      )
     )
    )
    Scale RGB=&VectorFromFloat(Value=1)
    ScaleA=true
    ScaleRGB=true
   )
   /SetVariables(Particles.Color02:LinearColor=&MultiplyLinearColor(
     Linear Color=$Particles.Color
     Scale Factor=&MakeVector4FromVectorAndFloat("Vector (RGB)"=<1 1 1>, "Float (Alpha)"=1.5)
    )) {isSetParameters=true}
   /SetVariables(Particles.Scale02:Vector3f=&Multiply_VectorByFloat(Vector=$Particles.Scale, Float=0.8)) {isSetParameters=true}
   &DynamicMaterialParameters(
    Index 1 Param 4=1
    Module.Param0WriteEnabled1=false
    Module.Param1WriteEnabled1=false
    Module.Param2WriteEnabled1=false
    Module.Param3WriteEnabled1=true
    Write Parameter Index 0=false
    Write Parameter Index 1=true
   )
  ]
  rendererClasses [&NiagaraMeshRendererProperties]
  renderers [
   {rendererClass=&NiagaraMeshRendererProperties, rendererIndex=0}
   {rendererClass=&NiagaraMeshRendererProperties, rendererIndex=1}
  ]
 }
 FireShell001 {
  emitterUpdateScript [
   &EmitterState(Life Cycle Mode=::NewEnumerator1, Loop Behavior=::NewEnumerator1, Loop Duration Mode=::NewEnumerator1, UseLoopDelay=false)
   &SpawnBurst_Instantaneous(
    Spawn Probability=&Lerp_Float(A=0, B=1, Alpha=&FN_FrontEndFloatModifier(Gameplay=1, Front End=1))
    Use Spawn Probability=false
    UseLoopCountLimit=false
   )
  ]
  particleSpawnScript [
   &PartitionParticles(Number Of Partitions=5)
   &InitializeParticle(
    Color=<rgba 0.0015 0.0152 0.0212 1>
    Mesh Scale=&Multiply_VectorByFloat(Vector=<1 1 1>, Float=0.18)
    Mesh Scale Mode=::NewEnumerator4
    Sprite Size Mode=::NewEnumerator3
    Uniform Sprite Size=65
    UsePositionOffset=false
    Write Material Random=true
   )
   &CameraOffset(Camera Offset Amount=10)
   &InitialMeshOrientation(
    Mesh Orientation Mode=::NewEnumerator2
    Orientation Vector=&RotateVectorByQuaternion(Vector To Rotate=<1 0 0>, Quaternion=$Engine.Owner.Rotation)
    Use Rotation Vector=false
   ) {enabled=false}
  ]
  particleUpdateScript [
   &SkeletalMeshLocation(
    Bone / Socket Index=$Particles.PartitionParticles.Partition
    Bone Sampling Mode=::NewEnumerator4
    MaskSpawnGroup=false
    Orientation Sampling=::NewEnumerator1
    Sampled Position Offset=&SelectVectorFromArray(
     Direct Array Index=$Particles.PartitionParticles.Partition
     Vector Selection Array=DataInterface() {propertyValues {FloatData [<0 0 0>, <0 0 0>]}}
    )
    Skeletal Mesh=$System.Skeletal Mesh
   )
   &ParticleState(Kill Particles When Lifetime Has Elapsed=false, Let Infinitely Lived Particles Die When Emitter Deactivates=true)
   &ScaleMeshSize(Scale Factor=&Multiply_VectorByFloat(Vector=<1 1 0.8>, Float=&FN_FrontEndFloatModifier(Gameplay=1, Front End=0.65)))
   &DynamicMaterialParameters(
    Index 0 Param 1=&SelectFloatFromArray(
     Direct Array Index=$Particles.PartitionParticles.Partition
     Float Selection Array=DataInterface() {propertyValues {FloatData [0.8,0.6,0.2,0.1]}}
    )
    Index 0 Param 2=$System.Dive
    Index 0 Param 4=0.5
    Index 1 Param 4=1
    Index 2 Param 1=0.8
    Index 3 Param 1=2
    Index 3 Param 2=$System.Timeline.NormalizedTime
    Module.Param0WriteEnabled=true
    Module.Param0WriteEnabled1=false
    Module.Param0WriteEnabled2=true
    Module.Param0WriteEnabled3=true
    Module.Param1WriteEnabled=true
    Module.Param1WriteEnabled1=false
    Module.Param1WriteEnabled2=true
    Module.Param1WriteEnabled3=true
    Module.Param2WriteEnabled=false
    Module.Param2WriteEnabled1=false
    Module.Param2WriteEnabled2=true
    Module.Param2WriteEnabled3=true
    Module.Param3WriteEnabled=true
    Module.Param3WriteEnabled1=true
    Module.Param3WriteEnabled2=true
    Module.Param3WriteEnabled3=true
    Write Parameter Index 1=true
    Write Parameter Index 2=true
    Write Parameter Index 3=true
   ) {module=DynamicMaterialParameters002}
   /SetVariables(Particles.Position=&TransformPosition(
     Position=&AddVectorToPosition(
      Position=Linked() {linkedVariable {name=System.SkullPosition, type=Vector3f}}
      Vector=&FN_FrontEndVector3Modifier(Gameplay=<0 0 -5>, Front End=<0 0 -3>)
     )
     Source Space=::World
     Destination Space=::Local
    )) {isSetParameters=true, module=SetVariables_#5}
   &ScaleColor(
    Scale Alpha=&Multiply_Float(
     A=&Multiply_Float(
      A=3
      B=&Power(
       a=&DotProduct(
        Vector A=&ReturnCameraVector(Camera Vectors=::NewEnumerator1, Coordinate Space=::Simulation)
        Vector B=<0 0 -1>
       )
       b=3
      )
     )
     B=&Lerp_Float(
      A=0
      B=1
      Alpha=&FloatFromCurve(
       FloatCurve=DataInterface() {propertyValues {Curve {
          defaultValue=0
          keys [
           {interpMode=RCIM_Cubic, tangentMode=RCTM_User, time=0, value=0}
           {arriveTangent=2.7352, interpMode=RCIM_Cubic, leaveTangent=2.7352, tangentMode=RCTM_User, time=1, value=1}
          ]
         }}}
       CurveIndex=$System.Timeline.NormalizedTime
      )
     )
    )
    ScaleA=true
    ScaleRGB=true
   ) {enabled=false, module=ScaleColor003}
   &ScaleColor(
    Scale Alpha=&Multiply_Float(
     A=&Multiply_Float(A=1, B=1)
     B=&Lerp_Float(
      A=0
      B=1
      Alpha=&FloatFromCurve(
       FloatCurve=DataInterface() {propertyValues {Curve {
          defaultValue=0
          keys [
           {interpMode=RCIM_Cubic, tangentMode=RCTM_User, time=0.5, value=0}
           {arriveTangent=2.7352, interpMode=RCIM_Cubic, leaveTangent=2.7352, tangentMode=RCTM_User, time=1, value=1}
          ]
         }}}
       CurveIndex=$System.Timeline.NormalizedTime
      )
     )
    )
    ScaleA=true
    ScaleRGB=true
   )
   /SetVariables(Particles.TopColor:LinearColor=&MultiplyLinearColor(
     Linear Color=$Particles.Color
     Scale Factor=&MakeVector4FromVectorAndFloat(
      "Vector (RGB)"=<1 1 1>
      "Float (Alpha)"=&FloatFromCurve(
       FloatCurve=DataInterface() {propertyValues {Curve {
          defaultValue=0
          keys [
           {arriveTangent=0.2035, interpMode=RCIM_Cubic, leaveTangent=0.2035, tangentMode=RCTM_User, time=0.5, value=0}
           {interpMode=RCIM_Cubic, tangentMode=RCTM_User, time=1, value=1.25}
          ]
         }}}
       CurveIndex=&DotProduct(
        Vector A=&ReturnCameraVector(
         Camera Vectors=::NewEnumerator1
         Coordinate Space=::Simulation
         CameraQuery=DataInterface() {propertyValues {bRequireCurrentFrameData=false}}
        )
        Vector B=<0 0 -1>
       )
       Scale Curve=2
      )
     )
    )) {isSetParameters=true}
   /SetVariables(Particles.BottomMeshPosition:NiagaraPosition=&AddVectorToPosition(
     Position=$Particles.Position
     Vector=&TransformVector(
      Vector=&FN_FrontEndVector3Modifier(Gameplay=<0 0 -24>, Front End=<0 0 -17>)
      Destination Space=::World
     )
    )) {isSetParameters=true}
   /SetVariables(Particles.BottomMeshScale:Vector3f=&Multiply_Vector(A=$Particles.Scale, B=&Multiply_VectorByFloat(Vector=<0.6 1 1>, Float=2))) {isSetParameters=true}
   /SetVariables(Particles.BottomMeshColor:LinearColor=&MultiplyLinearColor(
     Linear Color=<rgba 0.0911 0.4438 0.5833 1>
     Scale Factor=&MakeVector4FromVectorAndFloat(
      "Vector (RGB)"=&VectorFromFloat(Value=18)
      "Float (Alpha)"=&Lerp_Float(
       A=0
       B=1
       Alpha=&FloatFromCurve(
        FloatCurve=DataInterface() {propertyValues {Curve {
           defaultValue=0
           keys [
            {interpMode=RCIM_Cubic, tangentMode=RCTM_User, time=0.5, value=0}
            {arriveTangent=2.7352, interpMode=RCIM_Cubic, leaveTangent=2.7352, tangentMode=RCTM_User, time=1, value=1}
           ]
          }}}
        CurveIndex=$System.Timeline.NormalizedTime
       )
      )
     )
    )) {isSetParameters=true, module=SetVariables_#6}
   /SetVariables(Particles.BottomMeshColor:LinearColor=&MultiplyLinearColor(
     Linear Color=<rgba 0.0911 0.4438 0.5833 1>
     Scale Factor=&MakeVector4FromVectorAndFloat(
      "Vector (RGB)"=&VectorFromFloat(Value=18)
      "Float (Alpha)"=&Lerp_Float(
       A=0
       B=1
       Alpha=&FloatFromCurve(
        FloatCurve=DataInterface() {propertyValues {Curve {
           defaultValue=0
           keys [
            {interpMode=RCIM_Cubic, tangentMode=RCTM_User, time=0.5, value=0}
            {arriveTangent=2.7352, interpMode=RCIM_Cubic, leaveTangent=2.7352, tangentMode=RCTM_User, time=1, value=1}
           ]
          }}}
        CurveIndex=$System.Timeline.NormalizedTime
       )
      )
     )
    )) {isSetParameters=true}
   /SetVariables(Particles.TopScale:Vector3f=&Multiply_VectorByFloat(Vector=$Particles.Scale, Float=0.96)) {isSetParameters=true}
   /SetVariables(Particles.CoronaColor:LinearColor=&MultiplyLinearColor(Linear Color=<rgba 0.0015 0.0152 0.0212 1>, Scale Factor=<1 1 1 1>)) {isSetParameters=true}
   /SetVariables(Particles.Vector:Vector3f=&Multiply_Vector(A=$Particles.Scale, B=<1 1 1>)) {enabled=false, isSetParameters=true}
  ]
  rendererClasses [&NiagaraMeshRendererProperties]
  renderers [
   {rendererClass=&NiagaraMeshRendererProperties, rendererIndex=0}
   {rendererClass=&NiagaraMeshRendererProperties, rendererIndex=1}
  ]
 }
 HandTrails {
  emitterUpdateScript [
   &EmitterState(Life Cycle Mode=::NewEnumerator1, Loop Behavior=::NewEnumerator1, Loop Duration Mode=::NewEnumerator1, UseLoopDelay=false)
   &SpawnBurst_Instantaneous(
    Spawn Count=4
    Spawn Probability=&Lerp_Float(A=0, B=1, Alpha=&FN_FrontEndFloatModifier(Gameplay=1, Front End=1))
    Use Spawn Probability=false
    UseLoopCountLimit=false
   )
   /SetVariables(Emitter.Emissive:NiagaraFloat=&RemapRange(
     Input Value To Remap=&FloatFromCurve(
      FloatCurve=DataInterface() {propertyValues {Curve {
         defaultValue=1
         keys [
          {interpMode=RCIM_Cubic, tangentMode=RCTM_User, time=0, value=1}
          {interpMode=RCIM_Cubic, tangentMode=RCTM_User, time=0.5, value=0}
          {tangentMode=RCTM_User, time=1, value=1}
         ]
        }}}
      CurveIndex=&Subtract_Float(A=&Frac_Float(Float=&Multiply_Float(A=$Engine.Time, B=0.25)), B=0.05)
     )
     Input Min=0
     Input Max=1
     Output Min=0
     Output Max=&Lerp_Float(A=1.25, B=1.75, Alpha=$System.Timeline.NormalizedTime)
    )) {isSetParameters=true}
  ]
  particleSpawnScript [
   &PartitionParticles(Number Of Partitions=4)
   &InitializeParticle(
    Color=<rgba 0 0 0 1>
    Mesh Scale=<1.2 1 1>
    Mesh Scale Mode=::NewEnumerator4
    Sprite Size Mode=::NewEnumerator3
    Uniform Sprite Size=15
    UsePositionOffset=false
    Write Material Random=true
   )
   &InitialMeshOrientation(Mesh Orientation Mode=::NewEnumerator2, Rotation=<0 -0.25 0>, Use Rotation Vector=true)
   &InitialMeshOrientation(
    Mesh Orientation Mode=::NewEnumerator2
    Rotation=&SelectVectorFromArray(
     Direct Array Index=$Particles.PartitionParticles.Partition
     Vector Selection Array=DataInterface() {propertyValues {FloatData [<0 0 0>, <0 0 0.5>, <0 0 0>, <0 0 0.5>]}}
    )
    Use Rotation Vector=true
   ) {module=InitialMeshOrientation001}
   /SetVariables() {isSetParameters=true}
  ]
  particleUpdateScript [
   /SetVariables(Particles.isFootRibbon:NiagaraBool=&SetBoolByIntComparison(A=$Particles.PartitionParticles.Partition, B=1, Comparison Type=::NewEnumerator0)) {isSetParameters=true}
   &SkeletalMeshLocation(
    Bone / Socket Index=$Particles.PartitionParticles.Partition
    Bone Sampling Mode=::NewEnumerator4
    MaskSpawnGroup=false
    Orientation Sampling=::NewEnumerator1
    Sampled Position Offset=&SelectVectorFromArray(
     Direct Array Index=$Particles.PartitionParticles.Partition
     Vector Selection Array=DataInterface() {propertyValues {FloatData [<0 0 0>, <0 0 0>]}}
    )
    Skeletal Mesh=$System.Skeletal Mesh
   )
   &ScaleColor(
    Scale Alpha=&Lerp_Float(A=0.8, B=1, Alpha=$System.Timeline.NormalizedTime)
    Scale RGB=&VectorFromFloat(Value=1)
    ScaleA=true
    ScaleRGB=true
   )
   &ParticleState(Kill Particles When Lifetime Has Elapsed=false, Let Infinitely Lived Particles Die When Emitter Deactivates=true)
   &ScaleMeshSize(Scale Factor=&SelectVectorFromArray(
     Direct Array Index=$Particles.PartitionParticles.Partition
     Vector Selection Array=DataInterface() {propertyValues {FloatData [<0.8 1 1>, <0.8 1 1>, <0.4 0.4 0.4>, <0.4 0.4 0.4>]}}
    )) {module=ScaleMeshSize001}
   &ScaleMeshSize(Scale Factor=&MakeVector(
     X=&Lerp_Float(
      A=&FN_FrontEndFloatModifier(Gameplay=0.5, Front End=0.7)
      B=&FN_FrontEndFloatModifier(Gameplay=1, Front End=0.8)
      Alpha=&ClampFloat(Float=$System.FN_Contrail_Setup.Dive Value)
     )
     Y=1
     Z=1
    ))
   &DynamicMaterialParameters(
    Index 0 Param 1=&SelectFloatFromArray(
     Direct Array Index=$Particles.PartitionParticles.Partition
     Float Selection Array=DataInterface() {propertyValues {FloatData [0.8,0.6,0.2,0.1]}}
    )
    Index 0 Param 2=$System.Dive
    Index 0 Param 4=&Multiply_Float(A=&MakeCustomFloatFromBool(Bool=$Particles.isFootRibbon, True Float=0.6, False Float=0.5), B=1)
    Index 1 Param 4=1
    Index 2 Param 1=&MakeCustomFloatFromBool(Bool=$Particles.isFootRibbon, True Float=0.45, False Float=0.65)
    Index 3 Param 1=1.5
    Index 3 Param 2=$System.Timeline.NormalizedTime
    Module.Param0WriteEnabled=true
    Module.Param0WriteEnabled1=false
    Module.Param0WriteEnabled2=true
    Module.Param0WriteEnabled3=true
    Module.Param1WriteEnabled=true
    Module.Param1WriteEnabled1=false
    Module.Param1WriteEnabled2=true
    Module.Param1WriteEnabled3=true
    Module.Param2WriteEnabled=false
    Module.Param2WriteEnabled1=false
    Module.Param2WriteEnabled2=true
    Module.Param2WriteEnabled3=true
    Module.Param3WriteEnabled=true
    Module.Param3WriteEnabled1=true
    Module.Param3WriteEnabled2=true
    Module.Param3WriteEnabled3=true
    Write Parameter Index 1=true
    Write Parameter Index 2=true
    Write Parameter Index 3=true
   ) {module=DynamicMaterialParameters001}
   &FN_Mesh_Attributes(
    Color=&MultiplyLinearColor(
     Linear Color=$Particles.Color
     Scale Factor=&MakeVector4(
      X=1
      Y=1
      Z=1
      W=&Multiply_Float(
       A=&FloatFromCurve(
        FloatCurve=DataInterface() {propertyValues {Curve {
           defaultValue=0
           keys [
            {interpMode=RCIM_Cubic, tangentMode=RCTM_User, time=0, value=0}
            {arriveTangent=2.7352, interpMode=RCIM_Cubic, leaveTangent=2.7352, tangentMode=RCTM_User, time=1, value=1}
           ]
          }}}
        CurveIndex=$System.Timeline.NormalizedTime
       )
       B=&FloatFromCurve(
        FloatCurve=DataInterface() {propertyValues {Curve {
           defaultValue=0
           keys [
            {arriveTangent=0.2035, interpMode=RCIM_Cubic, leaveTangent=0.2035, tangentMode=RCTM_User, time=0.7, value=1}
            {interpMode=RCIM_Cubic, tangentMode=RCTM_User, time=1, value=0}
           ]
          }}}
        CurveIndex=&DotProduct(
         Vector A=&ReturnCameraVector(
          Camera Vectors=::NewEnumerator1
          Coordinate Space=::Simulation
          CameraQuery=DataInterface() {propertyValues {bRequireCurrentFrameData=false}}
         )
         Vector B=<0 0 -1>
        )
       )
      )
     )
    )
    Create Color Attribute=true
    Create Position Attribute=true
    Create Scale Attribute=true
    Position:Vector3f=&AddVectorToPosition(Position=$Particles.Position, Vector=<0 0 0>)
    Scale=&Multiply_VectorByFloat(
     Vector=&MakeCustomVectorFromBool(Bool=$Particles.isFootRibbon, True Vector=<1 0.7 0.7>, False Vector=<1.4 0.8 0.8>)
     Float=0.1
    )
    Write to Module Namespace=true
   )
   &FN_Mesh_Attributes(
    Color=&MultiplyLinearColor(
     Linear Color=$Particles.Color
     Scale Factor=&MakeVector4(
      X=1
      Y=1
      Z=1
      W=&Multiply_Float(
       A=&FloatFromCurve(
        FloatCurve=DataInterface() {propertyValues {Curve {
           defaultValue=0
           keys [
            {interpMode=RCIM_Cubic, tangentMode=RCTM_User, time=0, value=0}
            {arriveTangent=2.7352, interpMode=RCIM_Cubic, leaveTangent=2.7352, tangentMode=RCTM_User, time=1, value=1}
           ]
          }}}
        CurveIndex=$System.Timeline.NormalizedTime
       )
       B=&FloatFromCurve(
        FloatCurve=DataInterface() {propertyValues {Curve {
           defaultValue=0
           keys [
            {arriveTangent=0.2035, interpMode=RCIM_Cubic, leaveTangent=0.2035, tangentMode=RCTM_User, time=0.7, value=1}
            {interpMode=RCIM_Cubic, tangentMode=RCTM_User, time=1, value=0}
           ]
          }}}
        CurveIndex=&DotProduct(
         Vector A=&ReturnCameraVector(
          Camera Vectors=::NewEnumerator1
          Coordinate Space=::Simulation
          CameraQuery=DataInterface() {propertyValues {bRequireCurrentFrameData=false}}
         )
         Vector B=<0 0 -1>
        )
       )
      )
     )
    )
    Create Color Attribute=true
    Create Position Attribute=true
    Create Scale Attribute=true
    Position:Vector3f=&AddVectorToPosition(Position=$Particles.Position, Vector=<0 0 0>)
    Scale=&Multiply_VectorByFloat(
     Vector=&MakeCustomVectorFromBool(Bool=$Particles.isFootRibbon, True Vector=<1 0.7 0.7>, False Vector=<2 0.8 0.8>)
     Float=0.1
    )
    Write to Module Namespace=true
   ) {module=FN_Mesh_Attributes001}
   /SetVariables(Particles.MeshPosition:NiagaraPosition=&AddVectorToPosition(
     Position=$Particles.Position
     Vector=&TransformVector(
      Vector=&MakeCustomVectorFromBool(
       Bool=$Particles.isFootRibbon
       True Vector=&Lerp_Vector(A=<0 0 -5>, B=<0 0 0>, Alpha=$System.Timeline.NormalizedTime)
       False Vector=<0 0 -5>
      )
      Destination Space=::World
     )
    )) {isSetParameters=true}
   /SetVariables(Particles.MeshPosition01:NiagaraPosition=&AddVectorToPosition(
     Position=$Particles.Position
     Vector=&TransformVector(
      Vector=&MakeCustomVectorFromBool(Bool=$Particles.isFootRibbon, True Vector=<0 0 5>, False Vector=<0 0 -5>)
      Destination Space=::World
     )
    )) {isSetParameters=true}
   &FN_Light_Attributes(
    Color=<rgba 0.038 0.3808 0.5312 1>
    Create Color Attribute=true
    Create Exponent Attribute=true
    Create Radius Attribute=true
    Enable Particle Light Enable=true
    Exponent=1
    Light Enabled=true
    Radius=10
    Scale Color=20
   )
   /SetVariables(Particles.LightPos:NiagaraPosition=&AddVectorToPosition(Position=$Particles.Position, Vector=&TransformVector(Vector=<0 0 10>, Destination Space=::World))) {isSetParameters=true}
   /SetVariables(Particles.GlowDyn:Vector4f=&MakeVector4(
     X=0.3
     Y=0.8
     Z=&RandomRangeFloat(Minimum=0, Maximum=1)
     W=&RandomRangeFloat(Minimum=0, Maximum=1)
    )) {isSetParameters=true}
   /SetVariables(Particles.DecalColor:LinearColor=&MultiplyLinearColor(
     Linear Color=<rgba 0.038 0.3808 0.5312 1>
     Scale Factor=&MakeVector4(
      X=1
      Y=1
      Z=1
      W=&FloatFromCurve(
       FloatCurve=DataInterface() {propertyValues {Curve {keys [<at 0 0>,<at 1 1>]}}}
       CurveIndex=&DotProduct(Vector A=&ReturnCameraVector(Camera Vectors=::NewEnumerator1, Coordinate Space=::World), Vector B=<0 0 -1>)
      )
     )
    )) {isSetParameters=true}
   /SetVariables(Particles.SpriteCameraOffset=10) {isSetParameters=true}
   /SetVariables(Particles.GlowPosition:NiagaraPosition=&AddVectorToPosition(
     Position=$Particles.Position
     Vector=&TransformVector(Vector=&Lerp_Vector(A=<0 0 0>, B=<0 0 15>, Alpha=$System.Timeline.NormalizedTime), Destination Space=::World)
    )) {isSetParameters=true}
   /SetVariables(Particles.InnerColor:LinearColor=&MultiplyLinearColor(
     Linear Color=<rgba 0.2856 0.5758 0.7031 1>
     Scale Factor=&MakeVector4FromVectorAndFloat(
      "Vector (RGB)"=&VectorFromFloat(Value=&RemapRange(
        Input Value To Remap=&FloatFromCurve(
         FloatCurve=DataInterface() {propertyValues {Curve {
            defaultValue=1
            keys [
             {interpMode=RCIM_Cubic, tangentMode=RCTM_User, time=0, value=1}
             {arriveTangentWeight=0.1667, interpMode=RCIM_Cubic, leaveTangentWeight=0.0333, tangentMode=RCTM_User, time=0.4, value=0}
             {arriveTangentWeight=0.0745, interpMode=RCIM_Cubic, leaveTangentWeight=0.2981, tangentMode=RCTM_User, time=0.6, value=0}
             {tangentMode=RCTM_User, time=1, value=1}
            ]
           }}}
         CurveIndex=&Subtract_Float(A=&Frac_Float(Float=&Multiply_Float(A=$Engine.Time, B=0.25)), B=0.05)
        )
        Input Min=0
        Input Max=1
        Output Min=1
        Output Max=&Lerp_Float(A=4, B=5, Alpha=$System.Timeline.NormalizedTime)
       ))
      "Float (Alpha)"=&Lerp_Float(A=0.7, B=1, Alpha=$System.Timeline.NormalizedTime)
     )
    )) {isSetParameters=true}
  ]
  rendererClasses [
   &NiagaraMeshRendererProperties
   &NiagaraLightRendererProperties
   &NiagaraSpriteRendererProperties
  ]
  renderers [
   {rendererClass=&NiagaraMeshRendererProperties, rendererIndex=0}
   {rendererClass=&NiagaraMeshRendererProperties, rendererIndex=1}
   {rendererClass=&NiagaraLightRendererProperties, rendererIndex=2}
   {rendererClass=&NiagaraSpriteRendererProperties, rendererIndex=3}
   {rendererClass=&NiagaraMeshRendererProperties, rendererIndex=4}
   {rendererClass=&NiagaraMeshRendererProperties, rendererIndex=5}
  ]
 }
 Moon {
  emitterUpdateScript [
   &EmitterState(Life Cycle Mode=::NewEnumerator1, Loop Behavior=::NewEnumerator1, Loop Duration Mode=::NewEnumerator1, UseLoopDelay=false)
   &SpawnBurst_Instantaneous(Use Spawn Probability=false, UseLoopCountLimit=false)
  ]
  particleSpawnScript [
   &InitializeParticle(
    Mesh Scale Mode=::NewEnumerator3
    Mesh Uniform Scale=0.1
    Sprite Rotation Mode=::NewEnumerator2
    Sprite Size Mode=::NewEnumerator3
    Uniform Sprite Size=180
    UsePositionOffset=false
    Write Material Random=true
   )
   /SetVariables(Particles.Position=$System.SkullPosition) {enabled=false, isSetParameters=true}
   &CameraOffset(Camera Offset Amount=-100) {enabled=false}
  ]
  particleUpdateScript [
   &ParticleState(
    Kill Particles When Lifetime Has Elapsed=false
    Let Infinitely Lived Particles Die When Emitter Deactivates=true
    Loop Particles Lifetime=true
   )
   &ScaleSpriteSize(
    Scale Sprite Size Mode=::NewEnumerator0
    Uniform Scale Factor=&FN_FrontEndFloatModifier(
     Gameplay=&Lerp_Float(
      A=0.6
      B=1
      Alpha=&FloatFromCurve(
       FloatCurve=DataInterface() {propertyValues {Curve {keys [<at 0.5 0>,<at 1 1>]}}}
       CurveIndex=$System.Timeline.NormalizedTime
       Scale Curve=0.35
      )
     )
     Front End=&Lerp_Float(
      A=0.45
      B=0.7
      Alpha=&FloatFromCurve(
       FloatCurve=DataInterface() {propertyValues {Curve {keys [<at 0.5 0>,<at 1 1>]}}}
       CurveIndex=$System.Timeline.NormalizedTime
       Scale Curve=0.35
      )
     )
    )
   )
   &ScaleMeshSize(Scale Factor=&Multiply_VectorByFloat(
     Vector=<1 1 1>
     Float=&FN_FrontEndFloatModifier(
      Gameplay=&Multiply_Float(
       A=&FloatFromCurve(FloatCurve=DataInterface() {propertyValues {Curve {keys [<at 0 1>,<at 0.6 1.8>]}}}, CurveIndex=$System.Timeline.NormalizedTime)
       B=&Lerp_Float(
        A=1
        B=1
        Alpha=/Game/Effects/Niagara/Modules/Math/FN_CameraDistance(
         0Dist:NiagaraFloat=50
         1Dist:NiagaraFloat=100
         Camera Query:NiagaraDataInterfaceCamera=DataInterface()
         Position:NiagaraPosition=$Particles.Position
        )
       )
      )
      Front End=&Multiply_Float(
       A=&FloatFromCurve(FloatCurve=DataInterface() {propertyValues {Curve {keys [<at 0 1.2>,<at 0.6 1.5>]}}}, CurveIndex=$System.Timeline.NormalizedTime)
       B=0.8
      )
     )
    ))
   /SetVariables(Particles.Position=$System.SkullPosition) {isSetParameters=true}
   &DynamicMaterialParameters(
    Index 0 Param 1=&FloatFromCurve(
     FloatCurve=DataInterface() {propertyValues {Curve {defaultValue=1, keys [{tangentMode=RCTM_User, time=0, value=0}, {tangentMode=RCTM_User, time=0.8, value=1}]}}}
     CurveIndex=$System.Timeline.NormalizedTime
    )
    Index 0 Param 2=&FN_FrontEndFloatModifier(
     Gameplay=&FloatFromCurve(
      FloatCurve=DataInterface() {propertyValues {Curve {
         defaultValue=1
         keys [
          {interpMode=RCIM_Cubic, tangentMode=RCTM_User, time=0.5, value=1.15}
          {
           arriveTangent=636.9491
           arriveTangentWeight=1.4825
           interpMode=RCIM_Cubic
           leaveTangent=-604.1033
           leaveTangentWeight=0.7307
           tangentMode=RCTM_Break
           tangentWeightMode=RCTWM_WeightedBoth
           time=0.8
           value=2.75
          }
          <at 1 1.5 RCIM_Cubic>
         ]
        }}}
      CurveIndex=$System.Timeline.NormalizedTime
      Scale Curve=0.6
     )
     Front End=&FloatFromCurve(
      FloatCurve=DataInterface() {propertyValues {Curve {
         defaultValue=1
         keys [
          {interpMode=RCIM_Cubic, tangentMode=RCTM_User, time=0.5, value=1.15}
          {
           arriveTangent=18.5938
           arriveTangentWeight=0.7226
           interpMode=RCIM_Cubic
           leaveTangent=-24.8395
           leaveTangentWeight=0.5164
           tangentMode=RCTM_Break
           tangentWeightMode=RCTWM_WeightedBoth
           time=0.8
           value=2
          }
          <at 1 1.25 RCIM_Cubic>
         ]
        }}}
      CurveIndex=$System.Timeline.NormalizedTime
      Scale Curve=0.6
     )
    )
    Module.Index0Float0Write=true
    Module.Index0Float1Write=true
    Module.Index0Float2Write=true
    Module.Index0Float3Write=true
   )
   /SetVariables(Particles.GlowScale:Vector3f=&Multiply_VectorByFloat(Vector=&Multiply_Vector(A=$Particles.Scale, B=<1 1 0.95>), Float=1.2)) {isSetParameters=true}
   /SetVariables() {isSetParameters=true}
   /SetVariables() {isSetParameters=true}
   /SetVariables(
    Particles.AirglowColor:LinearColor=&MultiplyLinearColor(
     Linear Color=<rgba 0.1826 0.4249 0.5312 1>
     Scale Factor=&MakeVector4(
      X=1
      Y=1
      Z=1
      W=&Lerp_Float(
       A=&FloatFromCurve(
        FloatCurve=DataInterface() {propertyValues {Curve {
           defaultValue=1
           keys [
            {interpMode=RCIM_Cubic, tangentMode=RCTM_User, time=0, value=1}
            {arriveTangentWeight=0.1333, interpMode=RCIM_Cubic, leaveTangentWeight=0.0667, tangentMode=RCTM_User, time=0.4, value=0}
            {arriveTangentWeight=0.1296, interpMode=RCIM_Cubic, leaveTangentWeight=0.2592, tangentMode=RCTM_User, time=0.6, value=0}
            {tangentMode=RCTM_User, time=1, value=1}
           ]
          }}}
        CurveIndex=&Subtract_Float(A=&Frac_Float(Float=&Multiply_Float(A=$Engine.Time, B=0.25)), B=0.05)
        Scale Curve=0.2
       )
       B=&FloatFromCurve(
        FloatCurve=DataInterface() {propertyValues {Curve {
           defaultValue=1
           keys [
            <at 0.5 0.2 RCIM_Cubic>
            {
             arriveTangent=20.0207
             arriveTangentWeight=0.7848
             interpMode=RCIM_Cubic
             leaveTangent=-18.5384
             leaveTangentWeight=0.706
             tangentMode=RCTM_Break
             tangentWeightMode=RCTWM_WeightedBoth
             time=0.8
             value=1.2
            }
            <at 1 0.2 RCIM_Cubic>
           ]
          }}}
        CurveIndex=$System.Timeline.NormalizedTime
       )
       Alpha=&FloatFromCurve(
        FloatCurve=DataInterface() {propertyValues {Curve {defaultValue=0.5, keys [{tangentMode=RCTM_User, time=0, value=0}, {tangentMode=RCTM_User, time=1, value=1}]}}}
        CurveIndex=$System.Timeline.NormalizedTime
       )
      )
     )
    )
    Particles.AirglowDyn:Vector4f=&MakeVector4(X=2, Y=1, Z=1, W=1)
   ) {isSetParameters=true}
   /SetVariables(
    Particles.Airglow2CamOffset=30
    Particles.Airglow2Color:LinearColor=&MultiplyLinearColor(
     Linear Color=<rgba 0.06 0.8 1.2 1>
     Scale Factor=&MakeVector4(
      X=1
      Y=1
      Z=1
      W=&FloatFromCurve(
       FloatCurve=DataInterface() {propertyValues {Curve {keys [<at 0.5 0>,<at 1 1>]}}}
       CurveIndex=$System.Timeline.NormalizedTime
       Scale Curve=&FN_FrontEndFloatModifier(Gameplay=0.2, Front End=0.1)
      )
     )
    )
    Particles.Airglow2Dyn:Vector4f=&MakeVector4(X=2, Y=1, Z=1, W=1)
    Particles.Airglow2Size:Vector2f=&Multiply_Vector2D_ByFloat(Float=&FN_FrontEndFloatModifier(Gameplay=0.435, Front End=0.5), Vector2D=$Particles.SpriteSize)
   ) {isSetParameters=true}
   /SetVariables(
    Particles.Airglow3CamOffset=-40
    Particles.Airglow3Color:LinearColor=&MultiplyLinearColor(
     Linear Color=<rgba 0.0107 0.1075 0.15 1>
     Scale Factor=&MakeVector4(
      X=1
      Y=1
      Z=1
      W=&FloatFromCurve(FloatCurve=DataInterface() {propertyValues {Curve {keys [<at 0.5 1>,<at 1 1>]}}}, CurveIndex=$System.Timeline.NormalizedTime, Scale Curve=10)
     )
    )
    Particles.Airglow3Dyn:Vector4f=&MakeVector4(X=2, Y=1, Z=1, W=1)
    Particles.Airglow3Size:Vector2f=&FN_FrontEndVector2Modifier(
     Front End:Vector2f=&Multiply_Vector2D_ByFloat(
      Float=&FloatFromCurve(
       FloatCurve=DataInterface() {propertyValues {Curve {keys [<at 0.5 1>,<at 1 1>]}}}
       CurveIndex=$System.Timeline.NormalizedTime
       Scale Curve=&FN_FrontEndFloatModifier(Gameplay=0.6, Front End=0.6)
      )
      Vector2D=$Particles.SpriteSize
     )
     Gameplay:Vector2f=&Multiply_Vector2D_ByFloat(
      Float=&FloatFromCurve(
       FloatCurve=DataInterface() {propertyValues {Curve {keys [<at 0.5 0.75>,<at 1 1>]}}}
       CurveIndex=$System.Timeline.NormalizedTime
       Scale Curve=&FN_FrontEndFloatModifier(Gameplay=0.6, Front End=0.6)
      )
      Vector2D=$Particles.SpriteSize
     )
    )
   ) {isSetParameters=true}
  ]
  rendererClasses [&NiagaraMeshRendererProperties, &NiagaraSpriteRendererProperties]
  renderers [
   {rendererClass=&NiagaraMeshRendererProperties, rendererIndex=0}
   {rendererClass=&NiagaraMeshRendererProperties, rendererIndex=1}
   {rendererClass=&NiagaraSpriteRendererProperties, rendererIndex=2}
   {rendererClass=&NiagaraSpriteRendererProperties, rendererIndex=3}
   {rendererClass=&NiagaraSpriteRendererProperties, rendererIndex=4}
  ]
 }
 MoonTrail {
  emitterUpdateScript [
   &EmitterState(Life Cycle Mode=::NewEnumerator1, Loop Behavior=::NewEnumerator1, Loop Duration Mode=::NewEnumerator1, UseLoopDelay=false)
   &SpawnBurst_Instantaneous(
    Spawn Count=&MakeIntFromFloat(Float=&FN_FrontEndFloatModifier(Gameplay=5, Front End=0))
    Spawn Probability=&Lerp_Float(A=0, B=1, Alpha=&FN_FrontEndFloatModifier(Gameplay=1, Front End=1))
    Use Spawn Probability=false
    UseLoopCountLimit=false
   )
  ]
  particleSpawnScript [
   &PartitionParticles(Number Of Partitions=5)
   &InitializeParticle(
    Color=<rgba 0 0 0 1>
    Mesh Scale=&Multiply_VectorByFloat(Vector=<0.4 1 1>, Float=1)
    Mesh Scale Mode=::NewEnumerator4
    Sprite Size Mode=::NewEnumerator3
    Uniform Sprite Size=65
    UsePositionOffset=false
    Write Material Random=true
   )
   &InitialMeshOrientation(Mesh Orientation Mode=::NewEnumerator2, Rotation=<0 -0.25 0>, Use Rotation Vector=true)
   &InitialMeshOrientation(
    Mesh Orientation Mode=::NewEnumerator2
    Rotation=&SelectVectorFromArray(
     Direct Array Index=$Particles.PartitionParticles.Partition
     Vector Selection Array=DataInterface() {propertyValues {FloatData [<0 0 0>, <0 0 0.5>, <0 0 0>, <0 0 0.5>]}}
    )
    Use Rotation Vector=true
   ) {module=InitialMeshOrientation001}
   &CameraOffset(Camera Offset Amount=-30)
  ]
  particleUpdateScript [
   &SkeletalMeshLocation(
    Bone / Socket Index=$Particles.PartitionParticles.Partition
    Bone Sampling Mode=::NewEnumerator4
    MaskSpawnGroup=false
    Orientation Sampling=::NewEnumerator1
    Sampled Position Offset=&SelectVectorFromArray(
     Direct Array Index=$Particles.PartitionParticles.Partition
     Vector Selection Array=DataInterface() {propertyValues {FloatData [<0 0 0>, <0 0 0>]}}
    )
    Skeletal Mesh=$System.Skeletal Mesh
   )
   &ParticleState(Kill Particles When Lifetime Has Elapsed=false, Let Infinitely Lived Particles Die When Emitter Deactivates=true)
   &ScaleMeshSize(Scale Factor=&SelectVectorFromArray(
     Direct Array Index=$Particles.PartitionParticles.Partition
     Vector Selection Array=DataInterface() {propertyValues {FloatData [<0.7 0.4 0.4>, <0.5 0.8 0.8>, <0.5 0.6 0.6>, <0.8 0.3 0.3>, <1.2 1 1>]}}
    ))
   &DynamicMaterialParameters(
    Index 0 Param 1=&SelectFloatFromArray(
     Direct Array Index=$Particles.PartitionParticles.Partition
     Float Selection Array=DataInterface() {propertyValues {FloatData [0.8,0.6,0.4,0.1,0.2]}}
    )
    Index 0 Param 2=$System.Dive
    Index 0 Param 4=&MakeCustomFloatFromBool(Bool=$Particles.isFootRibbon, True Float=0.6, False Float=0.5)
    Index 1 Param 4=1
    Index 2 Param 1=&SelectFloatFromArray(
     Direct Array Index=$Particles.PartitionParticles.Partition
     Float Selection Array=DataInterface() {propertyValues {FloatData [0.7,0.8,0.4,0.5,1]}}
    )
    Index 3 Param 1=2
    Index 3 Param 2=$System.Timeline.NormalizedTime
    Module.Param0WriteEnabled=true
    Module.Param0WriteEnabled1=false
    Module.Param0WriteEnabled2=true
    Module.Param0WriteEnabled3=true
    Module.Param1WriteEnabled=true
    Module.Param1WriteEnabled1=false
    Module.Param1WriteEnabled2=true
    Module.Param1WriteEnabled3=true
    Module.Param2WriteEnabled=false
    Module.Param2WriteEnabled1=false
    Module.Param2WriteEnabled2=true
    Module.Param2WriteEnabled3=true
    Module.Param3WriteEnabled=true
    Module.Param3WriteEnabled1=true
    Module.Param3WriteEnabled2=true
    Module.Param3WriteEnabled3=true
    Write Parameter Index 1=true
    Write Parameter Index 2=true
    Write Parameter Index 3=true
   ) {module=DynamicMaterialParameters001}
   /SetVariables(Particles.Position=&AddVectorToPosition(
     Position=Linked() {linkedVariable {name=System.SkullPosition, type=Vector3f}}
     Vector=&TransformVector(
      Vector=&SelectVectorFromArray(
       Direct Array Index=$Particles.PartitionParticles.Partition
       Vector Selection Array=DataInterface() {propertyValues {FloatData [<-10 -18 15>, <10 18 15>, <18 -13 10>, <-20 12 0>, <0 0 0>]}}
      )
      Destination Space=::Simulation
     )
    )) {isSetParameters=true, module=SetVariables_#5}
   &ScaleColor(
    Scale Alpha=&FloatFromCurve(
     FloatCurve=DataInterface() {propertyValues {Curve {
        defaultValue=0
        keys [
         {interpMode=RCIM_Cubic, tangentMode=RCTM_User, time=0, value=0}
         {arriveTangent=2.7352, interpMode=RCIM_Cubic, leaveTangent=2.7352, tangentMode=RCTM_User, time=1, value=1}
        ]
       }}}
     CurveIndex=$System.Timeline.NormalizedTime
     Scale Curve=0.8
    )
    ScaleA=true
    ScaleRGB=true
   ) {module=ScaleColor001}
  ]
  rendererClasses [&NiagaraMeshRendererProperties]
  renderers [{rendererClass=&NiagaraMeshRendererProperties, rendererIndex=0}]
 }
 Sparkles {
  emitterUpdateScript [
   &EmitterState()
   &SpawnRate(
    SpawnRate=&Multiply_Float(
     A=&Lerp_Float(
      A=&FloatFromCurve(
       FloatCurve=DataInterface() {propertyValues {Curve {
          defaultValue=1
          keys [
           {interpMode=RCIM_Cubic, tangentMode=RCTM_User, time=0, value=1}
           {arriveTangentWeight=0.0833, interpMode=RCIM_Cubic, leaveTangentWeight=0.1667, tangentMode=RCTM_User, time=0.25, value=0}
           {arriveTangentWeight=0.2778, interpMode=RCIM_Cubic, leaveTangentWeight=0.1389, tangentMode=RCTM_User, time=0.75, value=0}
           {tangentMode=RCTM_User, time=1, value=1}
          ]
         }}}
       CurveIndex=&Subtract_Float(A=&Frac_Float(Float=&Multiply_Float(A=$Engine.Time, B=0.25)), B=0.05)
       Scale Curve=30
      )
      B=45
      Alpha=$System.Timeline.NormalizedTime
     )
     B=&FNTimeOfDayScalar(Night Scale=0.1, Morning Scale=1, Evening Scale=1)
    )
    Use Spawn Probability=false
   )
  ]
  particleSpawnScript [
   &PartitionParticles(Number Of Partitions=5)
   /SetVariables(Particles.isFoot:NiagaraBool=&Boolean_AndOperation(
     Bool1=&SetBoolByIntComparison(A=$Particles.PartitionParticles.Partition, B=1, Comparison Type=::NewEnumerator0)
     Bool2=&SetBoolByIntComparison(A=$Particles.PartitionParticles.Partition, B=4, Comparison Type=::NewEnumerator6)
    )) {isSetParameters=true}
   &InitializeParticle(
    Lifetime=&RandomRangeFloat(Minimum=0.6, Maximum=1.5)
    Position Mode=::NewEnumerator0
    Sprite Rotation Mode=::NewEnumerator2
    Sprite Size Mode=::NewEnumerator3
    Uniform Sprite Size=&Multiply_Float(
     A=&Multiply_Float(
      A=&RandomRangeFloat(Minimum=3, Maximum=4)
      B=&MakeCustomFloatFromBool(Bool=$Particles.isFoot, True Float=0.8, False Float=1)
     )
     B=&Lerp_Float(A=0.75, B=1, Alpha=$System.Timeline.NormalizedTime)
    )
    Write Material Random=true
   )
   &FN_SetSpawnGroup(Number of Spawn Groups=5)
  ]
  particleUpdateScript [
   &ParticleState()
   &SpriteRotationRate(Rotation Rate=&Multiply_Float(
     A=&UniformAOrBFloat(A=-1, Evaluation Type=::NewEnumerator0, Recalculate Random Each Loop=true, Override Seed=false)
     B=&RandomRangeFloat(Minimum=100, Maximum=250)
    ))
   &SkeletalMeshLocation(
    Bone / Socket Index=$Particles.SpawnGroup
    Bone Sampling Mode=::NewEnumerator4
    MaskSpawnGroup=false
    Sampled Position Offset=&RandomRangeVector(Minimum=<-8 -8 -8>, Maximum=<8 8 8>)
    Skeletal Mesh=$User.SkeletalMesh
   )
   /SetVariables(Particles.Position=/Niagara/DynamicInputs/TypeConversions/MakeCustomPositionFromBool(
     Bool:NiagaraBool=&SetBoolByIntComparison(A=$Particles.Initial.SpawnGroup, B=4, Comparison Type=::NewEnumerator4)
     True Position:NiagaraPosition=&TransformVector(
      Vector=&AddVectorToPosition(
       Position=Linked() {linkedVariable {name=System.SkullPosition, type=Vector3f}}
       Vector=&RandomRangeVector(Minimum=<-20 -20 -20>, Maximum=<20 20 20>)
      )
      Source Space=::World
      Destination Space=::Local
      Transform Vector as Position=true
     )
     False Position:NiagaraPosition=$Particles.Position
    )) {isSetParameters=true}
   &ScaleColor(
    Linear Color Curve=DataInterface() {propertyValues {
      AlphaCurve {keys [<at 0 1>,<at 1 0>]}
      BlueCurve {keys [<at 0 5>,<at 0.1051 1>,<at 0.2645 0.9531>,<at 0.517 0.7>]}
      GreenCurve {keys [<at 0 5>,<at 0.1051 1>,<at 0.2645 0.9531>,<at 0.517 0.7>]}
      RedCurve {keys [<at 0 5>,<at 0.1051 1>,<at 0.2645 0.9531>,<at 0.517 0.7>]}
     }}
    Scale Mode=::NewEnumerator2
   )
   &ScaleSpriteSize(Scale Factor=&Vector2DFromFloat(Value=&FloatFromCurve(
      FloatCurve=DataInterface() {propertyValues {Curve {keys [
          {arriveTangent=14.26, interpMode=RCIM_Cubic, leaveTangent=14.26, tangentMode=RCTM_User, time=0, value=0}
          {arriveTangentWeight=0.0869, interpMode=RCIM_Cubic, leaveTangentWeight=0.1958, tangentMode=RCTM_User, time=0.2, value=1}
          <at 1 0>
         ]}}}
      Scale Curve=1.8
     )))
   /SetVariables(Particles.Position=&AddVectorToPosition(
     Position=$Particles.Position
     Vector=&MakeVector(
      X=0
      Y=0
      Z=&FloatFromCurve(
       FloatCurve=DataInterface() {propertyValues {Curve {defaultValue=1, keys [{tangentMode=RCTM_User, time=0, value=0}, {tangentMode=RCTM_User, time=1, value=1}]}}}
       Scale Curve=&Multiply_Float(
        A=&Multiply_Float(
         A=&RandomRangeFloat(Minimum=80, Maximum=150)
         B=&Lerp_Float(A=1, B=1.5, Alpha=$System.Timeline.NormalizedTime)
        )
        B=&MakeCustomFloatFromBool(Bool=$Particles.isFoot, True Float=0.5, False Float=1)
       )
      )
     )
    )) {isSetParameters=true}
   /SetVariables(
    Particles.AirglowColor:LinearColor=&MultiplyLinearColor(Linear Color=$Particles.Color, Scale Factor=&MakeVector4(X=1, Y=1, Z=1, W=0.1))
    Particles.AirglowSize:Vector2f=&Multiply_Vector2D_ByFloat(Float=3, Vector2D=$Particles.SpriteSize)
   ) {isSetParameters=true}
  ]
  rendererClasses [&NiagaraSpriteRendererProperties]
  renderers [
   {rendererClass=&NiagaraSpriteRendererProperties}
   {rendererClass=&NiagaraSpriteRendererProperties, rendererIndex=1}
  ]
 }
}
summary {
 systemName=NS_Contrail_GrimHound
 emitters [
  {emitterName=Sparkles, rendererClasses [&NiagaraSpriteRendererProperties]}
  {emitterName=Moon, rendererClasses [&NiagaraMeshRendererProperties, &NiagaraSpriteRendererProperties]}
  {
   emitterName=HandTrails
   rendererClasses [
    &NiagaraMeshRendererProperties
    &NiagaraLightRendererProperties
    &NiagaraSpriteRendererProperties
   ]
  }
  {emitterName=MoonTrail, rendererClasses [&NiagaraMeshRendererProperties]}
  {emitterName=FireShell001, rendererClasses [&NiagaraMeshRendererProperties]}
  {emitterName=Corona, rendererClasses [&NiagaraMeshRendererProperties]}
 ]
 userVariables [
  User.bIsFrontEnd {type=NiagaraBool}
  User.bIsFrontEndPreview {type=NiagaraBool}
  User.FeetRestingPositions {
   type=NiagaraDataInterfaceArrayFloat3
   description=""
   defaultValue {
    struct=NiagaraExt_VariableValue_DataInterface
    value {
     dataInterface=&self:NiagaraDataInterfaceArrayFloat3_0
     dataInterfaceClass=&NiagaraDataInterfaceArrayFloat3
    }
   }
  }
  User.ForwardDot {type=NiagaraFloat, defaultValue=0, description=""}
  User.Landing {type=NiagaraBool}
  User.ParaGlideLeanAlpha {type=NiagaraFloat, defaultValue=0, description=""}
  User.PlayerSpeed {type=NiagaraFloat, defaultValue=0, description=""}
  User.RibbonIDOffset {type=NiagaraInt32}
  User.SkeletalMesh {
   type=NiagaraDataInterfaceSkeletalMesh
   defaultValue {
    struct=NiagaraExt_VariableValue_DataInterface
    value {dataInterface=&self:NiagaraDataInterfaceSkeletalMesh_0}
   }
  }
  User.SpawnCount {type=NiagaraInt32}
 ]
}
